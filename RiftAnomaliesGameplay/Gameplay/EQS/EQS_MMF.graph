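# EQS query in the Rift Anomalies gameplay content that finds widely spaced locations around a supplied world position, within the safe zone.
# A SimpleGrid generator centers on the world-location param context, with grid size bound to a query parameter.
# A Project test traces geometry on a game trace channel over a tall vertical range, then an InsideAthenaSafeZone test checks the storm circle; both filter and score.
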EnvQuery /RiftAnomaliesGameplay/Gameplay/EQS/EQS_MMF
options [{
  generator {
   name=EnvQueryGenerator_SimpleGrid_0
   class=EnvQueryGenerator_SimpleGrid
   properties {
    generateAround=/Script/FortniteAI.FortQueryContext_WorldLocationParam
    spaceBetween=15000
    gridSize {
     dataBinding=&self:EnvQueryGenerator_SimpleGrid_0.AIDataProvider_QueryParams_0
     dataField=FloatValue
     defaultValue=500
    }
    projectionData {extentX=0, navigationFilter=null, postProjectionVerticalOffset=0, projectDown=1024, projectUp=1024}
   }
  }
  tests [
   EnvQueryTest_Project_0 {
    class=EnvQueryTest_Project
    properties {
     scoringFactor=1
     testComment=""
     testPurpose=FilterAndScore
     projectionData {
      navigationFilter=null
      postProjectionVerticalOffset=0
      projectDown=50000
      projectUp=50000
      serializedChannel=ECC_GameTraceChannel11
      traceChannel=TraceTypeQuery10
      traceMode=GeometryByChannel
     }
    }
   }
   FortQueryTest_InsideAthenaSafeZone_0 {class=FortQueryTest_InsideAthenaSafeZone, properties {testPurpose=FilterAndScore}}
  ]
 }]
